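# M_ChromeBall is an engine material for the chrome ball of the editor's ColorCalibrator meshes, using the Default Lit shading model.
# It holds only three scalar constants with no textures or parameters; their links to material outputs are not shown.
# Planar forward reflections are enabled, along with usage flags for skeletal meshes, mesh deformers and static lighting.
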Material /Engine/EditorMeshes/ColorCalibrator/M_ChromeBall
graph {
 properties {
  bUsePlanarForwardReflections=true
  bUsedWithMeshDeformer=true
  bUsedWithSkeletalMesh=true
  bUsedWithStaticLighting=true
  pixelDepthOffsetMode=PDOM_Legacy
  usedShadingModels=Default Lit
 }
 nodes [
  Constant() {r=1}
  Constant() {r=1}
  Constant() {r=0.1}
 ]
}
