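# M_ClearCoat is an engine editor material for the ColorCalibrator meshes, using the Clear Coat shading model.
# Its graph holds only four scalar constants, with no textures, parameters or math, so it describes a fixed, uniform surface.
# The connections from those constants to material inputs are not recorded here.
# It is flagged for use with static lighting.
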Material /Engine/EditorMeshes/ColorCalibrator/M_ClearCoat
graph {
 properties {bUsedWithStaticLighting=true, pixelDepthOffsetMode=PDOM_Legacy, shadingModel=MSM_ClearCoat, usedShadingModels=Clear Coat}
 nodes [
  Constant() {r=1}
  Constant() {r=0.65}
  Constant()
  Constant() {r=0.18}
 ]
}
