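# M_ClearCoatBlack is a minimal engine material in the ColorCalibrator editor mesh folder, apparently a black surface for color calibration meshes.
# Its graph holds only two Constant nodes at their default value, with no textures, parameters or math, and no visible links to material inputs.
# Despite the ClearCoat name, the recorded shading model is Default Lit, and the material is flagged for use with static lighting.
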Material /Engine/EditorMeshes/ColorCalibrator/M_ClearCoatBlack
graph {
 properties {bUsedWithStaticLighting=true, pixelDepthOffsetMode=PDOM_Legacy, usedShadingModels=Default Lit}
 nodes [
  Constant()
  Constant()
 ]
}
