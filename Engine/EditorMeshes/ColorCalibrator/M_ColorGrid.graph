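# M_ColorGrid is a Default Lit material for the editor's color calibrator mesh, built around a color checker texture.
# It contains one Color_checker texture sample and two scalar constants, with no visible links showing which material inputs they feed.
# Usage flags allow skeletal meshes, mesh deformers and static lighting.
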Material /Engine/EditorMeshes/ColorCalibrator/M_ColorGrid
graph {
 properties {
  bUsedWithMeshDeformer=true
  bUsedWithSkeletalMesh=true
  bUsedWithStaticLighting=true
  pixelDepthOffsetMode=PDOM_Legacy
  usedShadingModels=Default Lit
 }
 nodes [
  Constant() {r=0.95}
  Constant() {r=0.2}
  TextureSample() {texture=&/Color_checker}
 ]
}
