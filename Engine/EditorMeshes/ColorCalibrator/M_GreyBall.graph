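# Default Lit material for the grey ball in the engine's ColorCalibrator editor meshes, by its name and path a neutral reference for judging lighting and color.
# The graph holds only three scalar constants, including the standard 18% middle-grey value, with no textures, parameters or math, and no connections to material inputs are shown.
# Usage flags cover static lighting, skeletal meshes and mesh deformers.
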Material /Engine/EditorMeshes/ColorCalibrator/M_GreyBall
graph {
 properties {
  bUsedWithMeshDeformer=true
  bUsedWithSkeletalMesh=true
  bUsedWithStaticLighting=true
  pixelDepthOffsetMode=PDOM_Legacy
  usedShadingModels=Default Lit
 }
 nodes [
  Constant() {r=0.18}
  Constant() {r=0.5}
  Constant() {r=1}
 ]
}
