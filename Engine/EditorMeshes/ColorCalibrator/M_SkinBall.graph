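# M_SkinBall is a subsurface-shaded material in the editor's ColorCalibrator folder, apparently a skin-tone reference ball for color calibration.
# It uses only constants: a pale skin color and a darker reddish-brown color, each dimmed by a multiply, plus a standalone scalar.
# It samples no textures and exposes no parameters, and it is flagged for use with static lighting.
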Material /Engine/EditorMeshes/ColorCalibrator/M_SkinBall
graph {
 properties {bUsedWithStaticLighting=true, pixelDepthOffsetMode=PDOM_Legacy, shadingModel=MSM_Subsurface, usedShadingModels=Subsurface}
 nodes [
  Constant() {r=0.5}
  Multiply(A=Constant3Vector() {constant=<rgba 0.5508 0.4313 0.3731 0>}, B=Constant() {r=0.7})
  Multiply(A=Constant3Vector() {constant=<rgba 0.4576 0.1383 0.0546 0>}) {constB=0.5}
 ]
}
